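# Translucent two-sided unlit material for the CoyoTear pickaxe loop trail effect on Niagara ribbons and mesh particles.
# Three panned textures form the streaks, colored by a curve atlas along U and particle color, with lead mask and optional soft-glow trail and depth fade.
# An Appear dynamic parameter drives an optional wipe with a bright band.
# World position offset combines SplineThicken, using dynamic widths, with masked sine-and-noise wobble.
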Material /Game/Weapons/FORT_Melee/Pickaxe_CoyoTear/FX/Materials/M_Pickaxe_CoyoTear_LoopTrail
graph {
 properties {
  bUsedWithNiagaraMeshParticles=true
  bUsedWithNiagaraRibbons=true
  blendMode=BLEND_Translucent
  pixelDepthOffsetMode=PDOM_Legacy
  shadingModel=MSM_Unlit
  twoSided=true
  usedShadingModels=Unlit
 }
 nodes [
  ComponentMask(None=VectorParameter() {parameterName=WPO UV Scale, defaultValue=<rgba 1 1 0 0>}.RGB) {g=true, r=true}
  n1 = ParticleColor()
  n2 = Panner(
   Coordinate=Add(A=TextureCoordinate() {vTiling=1.5}, B=Constant2Vector() {g=-0.2})
   Speed=AppendVector(
    A=ScalarParameter() {defaultValue=1, group=Panner Speeds, parameterName=TEX_2_Speed_X}
    B=ScalarParameter() {group=Panner Speeds, parameterName=TEX_2_Speed_Y}
   )
  ) {speedX=1}
  n3 = Multiply(A=Power(Base=TextureSample(UVs=@n2) {texture=/Game/Weapons/FORT_Melee/Pickaxe_Future_Samurai_Summer_Male/FX/T_Pickaxe_Future_Samurai_Summer_Male_Trail_01}.R) {constExponent=2.5})
  n4 = Panner(
   Coordinate=Add(A=TextureCoordinate() {vTiling=0.8}, B=Constant2Vector() {g=0.15})
   Speed=AppendVector(
    A=ScalarParameter() {defaultValue=0.4, group=Panner Speeds, parameterName=TEX_1_Speed_X}
    B=ScalarParameter() {group=Panner Speeds, parameterName=TEX_1_Speed_Y}
   )
  ) {speedX=0.4}
  n5 = Multiply(A=Power(Base=TextureSample(UVs=@n4) {samplerType=SAMPLERTYPE_LinearColor, texture=/Game/Effects/Fort_Effects/Textures/T_Contrail_StinkyCloud_Ribbon}.G) {constExponent=3}) {constB=1.5}
  n6 = Round(None=@n5)
  n7 = Add(A=LinearInterpolate(A=Round(None=@n3), B=@n3) {constAlpha=1}, B=Multiply(A=LinearInterpolate(A=@n6, B=@n5) {constAlpha=1}) {constB=0.3}) {constB=0}
  n8 = Panner(
   Coordinate=Add(A=TextureCoordinate() {uTiling=1.5}, B=Constant2Vector())
   Speed=AppendVector(
    A=ScalarParameter() {defaultValue=1, group=Panner Speeds, parameterName=TEX_3_Speed_X}
    B=ScalarParameter() {group=Panner Speeds, parameterName=TEX_3_Speed_Y}
   )
  ) {speedX=1}
  n9 = Multiply(A=Power(Base=TextureSample(UVs=@n8) {texture=/Game/Weapons/FORT_Melee/Pickaxe_Sunlit/FX/T_Pickaxe_Sunlit_Trail_01}.R) {constExponent=1})
  n10 = LinearInterpolate(A=Round(None=@n9), B=@n9) {constAlpha=1}
  n11 = LinearInterpolate(A=LinearInterpolate(A=@n7, B=Saturate(None=@n10)), B=NamedRerouteUsage(), Alpha=ScalarParameter() {parameterName=Use Soft Trail})
  n12 = NamedRerouteUsage()
  n13 = ScalarParameter() {parameterName=Use appear}
  n14 = ComponentMask(None=TextureCoordinate()) {r=true}
  n15 = Saturate(None=Multiply(A=@n14) {constB=3})
  n16 = OneMinus(None=ComponentMask(None=TextureCoordinate()) {r=true})
  n17 = Saturate(None=Multiply(A=@n16) {constB=5})
  n18 = Multiply(
   A=Multiply(A=Multiply(A=@n1.A, B=@n11), B=LinearInterpolate(B=@n12, Alpha=@n13) {constA=1})
   B=LinearInterpolate(B=Multiply(A=@n15, B=@n17) {constA=1}, Alpha=ScalarParameter() {defaultValue=1, parameterName=Use lead mask}) {constA=1}
  )
  LinearInterpolate(
   A=@n18
   B=DepthFade(Opacity=@n18, FadeDistance=ScalarParameter() {defaultValue=5, group=Depth Fade, parameterName=FadeDistance})
   Alpha=ScalarParameter() {group=Depth Fade, parameterName=USE_DEPTH_FADE}
  )
  n19 = LinearInterpolate(B=NamedRerouteUsage(), Alpha=ScalarParameter() {parameterName=Use appear})
  n20 = TextureCoordinate()
  n21 = CurveAtlasRowParameter(CurveTime=OneMinus(None=ComponentMask(None=@n20) {r=true})) {
   atlas=/Game/Weapons/FORT_Melee/Pickaxe_CoyoTear/FX/CA_Pickaxe_CoyoTear_Spline_Color
   curve=/Game/Weapons/FORT_Melee/Pickaxe_CoyoTear/FX/CV_Pickaxe_CoyoTear_Spline_Color
   parameterName=CV_Color
  }
  n22 = Multiply(A=@n21)
  n23 = Constant3Vector() {constant=<rgba 5 4.7508 1.172 1>}
  n24 = Add(A=Multiply(A=Multiply(A=@n22, B=@n23), B=@n7), B=Multiply(A=Multiply(A=@n21) {constB=0.5}, B=@n10)) {constB=0}
  n25 = Multiply(A=LinearInterpolate(A=Add(A=@n24) {constB=0}, B=@n1.RGB, Alpha=ScalarParameter() {parameterName=Use Soft Trail}), B=@n1.RGB)
  MaterialFunctionCall(Emissive=LinearInterpolate(A=Add(A=@n19, B=@n25), B=ParticleColor().RGB, Alpha=ScalarParameter() {group=Overrides, parameterName=ParticleColor_Override})) {materialFunction=&MF_FX_EmissiveAdjustments}
  n26 = TextureCoordinate() {uTiling=0.8}
  n27 = AppendVector(
   A=ScalarParameter() {defaultValue=0.4, group=Panner Speeds, parameterName=TEX_4_Speed_X}
   B=ScalarParameter() {group=Panner Speeds, parameterName=TEX_4_Speed_Y}
  )
  NamedRerouteDeclaration(None=TextureSample(UVs=Panner(Coordinate=@n26, Speed=@n27) {speedX=0.4}) {samplerType=SAMPLERTYPE_Grayscale, texture=&T_Smoke_Trail_Soft}.R) {name=Soft glow, nodeColor=<rgba 1 0.5029 0 1>}
  n28 = DynamicParameter() {defaultValue=<rgba 15 1 1 1>, paramNames [WidthBase,WidthTip,Appear,Param4]}
  n29 = ScalarParameter() {defaultValue=0.5, parameterName=WPO Width}
  n30 = TextureCoordinate()
  n31 = Power(Base=Sine(None=ComponentMask(None=@n30) {r=true}) {period=2}, Exp=ScalarParameter() {parameterName=Edge mask power})
  n32 = ScalarParameter() {defaultValue=10, parameterName=Width mask range}
  n33 = LinearInterpolate(A=ScalarParameter() {defaultValue=0.2, parameterName=Width min}, Alpha=Saturate(None=Multiply(A=@n31, B=@n32) {constB=10})) {constA=0.2}
  n34 = Multiply(A=Multiply(A=@n28.WidthBase, B=@n29), B=@n33)
  n35 = Multiply(A=Multiply(A=@n29, B=@n28.WidthTip), B=@n33)
  n36 = StaticBool()
  n37 = ScalarParameter() {defaultValue=1, parameterName=WPO UVScale}
  n38 = ComponentMask(None=Multiply(A=TextureCoordinate() {uTiling=1.5}, B=@n37)) {r=true}
  n39 = Multiply(A=Time()) {constB=0.6}
  n40 = Panner(
   Coordinate=Multiply(A=@n37, B=TextureCoordinate() {uTiling=4, vTiling=0})
   Speed=AppendVector(
    A=ScalarParameter() {defaultValue=0.8, group=Panner Speeds, parameterName=WPO_TEX_Speed_X}
    B=ScalarParameter() {group=Panner Speeds, parameterName=WPO_TEX_Speed_Y}
   )
  ) {speedX=0.8}
  n41 = Add(
   A=Sine(None=Add(A=@n38, B=@n39)) {period=0.7}
   B=LinearInterpolate(Alpha=TextureSample(UVs=@n40) {samplerType=SAMPLERTYPE_Grayscale, texture=&LowResBlurredNoise}.R) {constA=-0.2}
  )
  n42 = Multiply(A=Constant3Vector() {constant=<rgba 0 0 1 0>}, B=@n41)
  n43 = Multiply(A=Constant3Vector() {constant=<rgba 0 1 0 0>}, B=@n41)
  n44 = Multiply(A=Add(A=Transform(None=@n42), B=Transform(None=@n43)), B=ScalarParameter() {defaultValue=0.8, parameterName=WPO Internity})
  n45 = SphereMask(A=TextureCoordinate() {vTiling=0}, B=Constant2Vector() {r=0.5}) {attenuationRadius=0.5, hardnessPercent=0}
  n46 = OneMinus(None=ComponentMask(None=TextureCoordinate() {vTiling=0}) {r=true})
  n47 = Saturate(None=Multiply(A=Multiply(A=@n45, B=@n46)) {constB=1.5})
  NamedRerouteDeclaration(None=Add(
    A=MaterialFunctionCall(WidthBase=@n34, WidthTip=@n35, Expand U or V UV Channel=@n36) {materialFunction=&SplineThicken}.WorldPositionOffset
    B=Multiply(A=@n44, B=@n47)
   )) {name=WPO, nodeColor=<rgba 0.1807 0 1 1>}
  n48 = DynamicParameter() {defaultValue=<rgba 15 1 1 1>, paramNames [WidthBase,WidthTip,Appear,Param4]}
  n49 = Add(A=ComponentMask(None=TextureCoordinate()) {r=true}, B=LinearInterpolate(Alpha=@n48.Appear) {constA=-1}) {constB=0}
  n50 = Multiply(A=ComponentMask(None=TextureCoordinate()) {r=true})
  NamedRerouteDeclaration(None=Multiply(A=Saturate(None=@n49), B=Saturate(None=@n50))) {name=Appear, nodeColor=<rgba 1 0.2578 0 1>}
  n51 = VectorParameter() {parameterName=Appear Emissive Bright color, defaultValue=<rgba 1 1 1 1>}
  n52 = Add(A=ComponentMask(None=TextureCoordinate()) {r=true}, B=LinearInterpolate(Alpha=@n48.Appear) {constA=-1}) {constB=0}
  n53 = Multiply(B=Saturate(None=Sine(None=@n52) {period=2})) {constA=1}
  NamedRerouteDeclaration(None=Multiply(A=Multiply(A=@n51.RGB, B=@n53), B=ParticleColor().RGB)) {name=Appear Emissive Bright, nodeColor=<rgba 1 0.3064 0.0442 1>}
  NamedRerouteUsage()
  Power(Base=@n14) {constExponent=1.2}
  Power(Base=@n16) {constExponent=1.2}
 ]
}
